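# Animation Blueprint for the Galileo Kayak head that adapts it to the matching Kayak cape backpack.
# On customization load it detects that pack and derives hood state and helmet_correction shrink alpha from its "Hide Element 01" material parameter.
# The copied parent pose gets head_alt driven by head-versus-neck_02 rotation, damped to 30% when the hood is up.
# head_alt is slightly shrunk, with added yaw while skydive floating with the cape.
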AnimBlueprint /Game/Characters/Player/Male/Medium/Heads/M_MED_Galileo_Kayak_Head/Meshes/Male_Medium_Galileo_Kayak_Head_AnimBP
target_skeleton=&/Male_Medium_Galileo_Kayak_Head_Skeleton
graphs [
 AnimGraph {nodes [
   n1 = "Transform (Modify) Bone - Bone: head_alt"(
    ComponentPose=Local To Component(LocalPose=Use cached pose 'initial Pose'())
    Alpha=1
    Translation="0,0,0"
    Rotation=Get head alt rot() -> (head alt rot:Rotator)
    Scale="1,1,1"
   ) {
    rotationMode=BMM_Replace
    rotationSpace=BCS_BoneSpace
    translationSpace=BCS_BoneSpace
    alphaBoolBlend {blendOption=Linear}
    alphaScaleBiasClamp {interpSpeedDecreasing=1, interpSpeedIncreasing=1}
    boneToModify {boneName=head_alt}
   }
   n2 = Component To Local(ComponentPose=@n1)
   n3 = "Transform (Modify) Bone - Bone: head_alt"(
    ComponentPose=Local To Component(LocalPose=Use cached pose 'initial Pose'())
    Alpha=1
    Translation="0,0,0"
    Rotation=Get head rot() -> (head rot:Rotator)
    Scale="1,1,1"
   ) {
    rotationMode=BMM_Replace
    rotationSpace=BCS_BoneSpace
    translationSpace=BCS_BoneSpace
    alphaBoolBlend {blendOption=Linear}
    alphaScaleBiasClamp {interpSpeedDecreasing=1, interpSpeedIncreasing=1}
    boneToModify {boneName=head_alt}
   }
   n4 = Component To Local(ComponentPose=@n3)
   n5 = Get hood up() -> (hood up:bool)
   n6 = "Transform (Modify) Bone - Bone: head_alt"!(
    ComponentPose=Local To Component(LocalPose=Blend Poses by bool(BlendPose_0=@n2, BlendPose_1=@n4, bActiveValue=@n5) {blendType=Linear})
    Alpha=1
   ) -> (Pose) {scaleMode=BMM_Replace, alphaBoolBlend {blendOption=Linear}, boneToModify {boneName=head_alt}, scale=<0.93 0.93 0.93>}
   n7 = "Transform (Modify) Bone - Bone: head_alt"!(
    ComponentPose=@n6
    bAlphaBoolEnabled:bool=AND Boolean(A=Get bIsSkydiveFloating(), B=Get kayak cape() -> (kayak cape:bool))
    Translation:Vector="0,0,0"
    Rotation:Rotator="-0.000,25,-0.000"
    Scale:Vector="1,1,1"
   ) -> (Pose) {
    alphaInputType=Bool
    rotationMode=BMM_Additive
    rotationSpace=BCS_BoneSpace
    alphaBoolBlend {blendOption=Linear}
    boneToModify {boneName=head_alt}
    rotation {yaw=25}
   }
   Output Pose(Result=Component To Local(ComponentPose=@n7)) {layerGroup=None}
   n8 = "Transform (Modify) Bone - Bone: helmet_correction"(
    ComponentPose=Local To Component(LocalPose=Copy Pose From Mesh() {bUseAttachedParent=true})
    Alpha=Get backpack alpha() -> (backpack alpha:double)
    Translation="0,0,0"
    Rotation="0,0,0"
    Scale="0.8,0.750,0.750"
   ) {
    scaleMode=BMM_Replace
    scaleSpace=BCS_BoneSpace
    alphaBoolBlend {blendOption=Linear}
    boneToModify {boneName=helmet_correction}
    scale=<0.8 0.75 0.75>
   }
   Save cached pose 'initial Pose'(Pose=Component To Local(ComponentPose=@n8)) {cacheName=initial Pose}
  ]}
 EventGraph {nodes [
   Event BlueprintUpdateAnimation()
   n1 = GetOwnerSkeletalMeshComponent()
   n2 = BreakRotator(InRot=BreakTransform(InTransform=GetDeltaTransformFromRefPose(self=@n1, BoneName=head, BaseName=neck_02)).Rotation)
   > Set head alt rot(head alt rot:Rotator=MakeRotator(Roll=float * float(A=@n2, B=0.3), Pitch=float * float(A=@n2.Pitch, B=0.3), Yaw=float * float(A=@n2.Yaw, B=0.3))) -> (Output_Get:Rotator)
   > Set head rot(head rot:Rotator=MakeRotator(Roll=@n2, Pitch=@n2.Pitch, Yaw=@n2.Yaw)) -> (Output_Get:Rotator)
   Event BlueprintInitializeAnimation()
   n3 = TryGetPawnOwner()
   > Is Valid(InputObject=@n3)
   > n4 = Cast To FortPlayerPawnAthena(Object=@n3)
   > Set player_pawn(player_pawn:Fort Player Pawn Athena Object Reference=@n4) -> (Output_Get:Fort Player Pawn Athena Object Reference)
   Event OnCharacterCustomizationLoaded()
   > Do Once()
   n5 = GetSkeletalMeshForPartType(self=Get player_pawn() -> (player_pawn:Fort Player Pawn Athena Object Reference), PartType=Backpack)
   > Set backpack mesh(backpack mesh:Skeletal Mesh Component Object Reference=@n5) -> (Output_Get:Skeletal Mesh Component Object Reference)
   > Set kayak cape(kayak cape:bool="Equal (Object)"(
     A=Get SkeletalMesh!(self:Skinned Mesh Component Object Reference=@n5) -> (SkeletalMesh:Skeletal Mesh Object Reference)
     B=/Game/Accessories/FORT_Capes/M_MED_Galileo_Kayak_Cape/Meshes/M_MED_Galileo_Kayak_Pack
    ) {is=title K2Node_CallFunction}) -> (Output_Get:bool)
   > Branch(Condition=Get kayak cape())
   > n6 = Get Dynamic Material Instance(SkeletalMeshComponent=Get backpack mesh() -> (backpack mesh:Skeletal Mesh Component Object Reference))
   > n7 = Get Scalar Parameter Value!(execute, self:Material Instance Dynamic Object Reference=@n6, ParameterName:Name=Hide Element 01) -> (then, ReturnValue:float) {is=title K2Node_CallFunction}
   > n8 = Set backpack alpha(backpack alpha:double=float - float(B=@n7) {is=title K2Node_CallFunction}) -> (Output_Get:double)
   > Set hood up(hood up:bool="float >= float"(A=@n8, B=0.9) {is=title K2Node_CallFunction}) -> (Output_Get:bool)
   Sequence!(execute) -> (then_0, then_1, then_2:Exec)
  ]}
]
variables [
 backpack alpha {type=number}
 backpack mesh {type=SkeletalMeshComponent}
 head alt rot {type=Rotator}
 head rot {type=Rotator}
 hood up {type=boolean}
 kayak cape {type=boolean}
 player_pawn {type=FortPlayerPawnAthena}
]
